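# Animation Blueprint for the Ashton Chicago gadget prop skeleton, used with the male medium main-player combat gadget animations.
# A single sequence player supplies the default weapon pose and feeds the DefaultSlot, so montages can play over that base pose before reaching the output.
# The event graph is empty and there are no variables, so it holds no logic of its own.
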AnimBlueprint /Game/Animation/Game/MainPlayer/Combat/Gadgets/Medium/Male/Ashton/Chicago/AshtonChicago_Prop_AnimBP
target_skeleton=/Game/Athena/Items/Gameplay/BackPacks/Ashton/Chicago/Meshes/Ashton_Chicago_Gadget_Skeleton
graphs [
 AnimGraph {nodes [Output Pose(Result=Slot 'DefaultSlot'(Source=Sequence Player 'AshtonChicago_Default_Pose_Weapon'() {sequence=&/AshtonChicago_Default_Pose_Weapon})) {layerGroup=None}]}
 EventGraph {links [], nodes []}
]
variables []
